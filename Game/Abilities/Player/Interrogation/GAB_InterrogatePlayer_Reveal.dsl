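; Gameplay ability for the reveal phase of interrogating a player or NPC, which marks targets for the interrogator's squad.
; Event activation sets camera, holster and reticle state, plays a montage and starts a timer that triggers Reveal, then an end montage.
; Player targets expose their non-immune teammates, found via minigame or squad lookup; NPC targets expose nearby faction-filtered pawns, chests and gameplay items.
; Marks use timed indicators and stencils, and quest stats update.
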(event Custom|OnDied_Event_1
  (params (DamagedActor :type Actor%20Object%20Reference) (Damage :type Float%20%28single-precision%29) (InstigatedBy :type Controller%20Object%20Reference) (DamageCauser :type Actor%20Object%20Reference) (HitLocation :type Vector) (FHitComponent :type Primitive%20Component%20Object%20Reference) (BoneName :type Name) (Momentum :type Vector))
  ;[1 1 1] End the ability if player dies
  (Ability|EndAbility))

(event Ability|EventActivateAbilityFromEvent
  (params (EventData :type Gameplay%20Event%20Data%20Structure%20%28by%20ref%29))
  (bind (_eventtag _instigator _target _optionalobject) (Utilities|Struct|BreakGameplayEventData EventData))
  (bind _playerpawn (Variables|Default|GetPlayerPawn))
  ;[0.3672 1 0.3979] Start Ability
  (Utilities|IsValid _optionalobject
    (:Is%20Valid
      (bind _asfort_player_pawn (Utilities|Casting|CastToFortPlayerPawn _optionalobject)
        (:then
          (bind _output_get (Variables|Default|SetInterrogatedPlayer _asfort_player_pawn))
          (CallFunction|RPCSetInterrogatedPlayer _output_get)
          (CallFunction|InitAbility)
          (Utilities|FlowControl|Sequence
            (:then_0
              ;[1 1 1] End the ability if player dies
              (Utilities|IsValid _playerpawn
                (:Is%20Valid
                  (EventDispatchers|AssignOnDied _playerpawn (delegate Custom|OnDied_Event_1)))))
            (:then_1
              ;[1 1 1] Play Reveal Montage
              (Ability|Tasks|PlayMontageandWaitforTrigger "None" (Variables|Default|GetAnim_Reveal))
              (bind _returnvalue (Utilities|Time|SetTimerbyEvent (EventDispatchers|CreateEvent :!event Custom|EndInterrogation) (Variables|Default|GetRevealDuration)))
              (Variables|Default|SetTimer_Reveal _returnvalue))))
        (:CastFailed
          _)))
    (:Is%20Not%20Valid
      _))
  (Ability|EndAbility))

(event Ability|EventOnEndAbility
  (params (bWasCancelled :type Boolean))
  ;[1 0.4227 0.01201] End Ability
  (|CallParentFunction bWasCancelled)
  (Default|EndAbilityCleanup))

(event Custom|RPC_SetInterrogatedPlayer
  (params (InterrogatedPlayer :type Fort%20Player%20Pawn%20Object%20Reference))
  (Variables|Default|SetInterrogatedPlayer InterrogatedPlayer))

(event Custom|EndInterrogation
  ;[1 1 1] Mark Objects and End Ability
  (if (not (Variables|Default|GetEnding))
    (Variables|Default|SetEnding true)
    (Utilities|Time|ClearandInvalidateTimerbyHandle (Variables|Default|GetTimer_Reveal))
    (Ability|SetCanbeCanceled false)
    (Camera|PopCameraMode)
    (CallFunction|Reveal)
    (Ability|Tasks|WaitDelay (Variables|Default|GetSweepDelay)
      (:OnFinish
        (Ability|Tasks|PlayMontageandWaitforTrigger "None" (Variables|Default|GetAnim_End) 0.8
          (:Cancelled
            _)
          (:Completed
            _))
        (Ability|EndAbility)))))

(fn EndAbilityCleanup
  (bind _playerpawn (Variables|Default|GetPlayerPawn))
  (CallFunction|ToggleReticleVisibility false)
  (Camera|PopCameraMode)
  (Utilities|Time|ClearandInvalidateTimerbyHandle (Variables|Default|GetTimer_Reveal))
  (Utilities|IsValid _playerpawn
    (:Is%20Valid
      (Weapon|SetHolsterWeapon _playerpawn 0 false true (Variables|Default|GetRevealHolster)))))

(fn GAB_InterrogatePlayer_Reveal_AutoGenFunc
  (params (IndicatedActor :type Actor%20Object%20Reference)))

(fn GetTeamMembers
  (params (Target :type Pawn%20Object%20Reference) (bIncludeSelf :type Boolean))
  (bind _target (Variables|GetTarget))
  (bind _playerstate (Variables|/Script/Engine.Pawn|GetPlayerState _target))
  (bind _asfort_game_state_athena (Utilities|Casting|CastToFortGameStateAthena (Game|GetGameState))
    (:then
      (if (Class|FortGameStateAthena|IsCreativeorPlaygroundModePlaylist _asfort_game_state_athena)
        ;[0 1 0.2903] Creative Playlists
        (Utilities|IsValid _playerstate
          (:Is%20Valid
            (bind _returnvalue (VolumeManager|GetMinigameforPlayerState _playerstate))
            (Utilities|IsValid _returnvalue
              (:Is%20Valid
                (Utilities|IsValid (Minigame|GetMinigameTeamComponent _returnvalue)
                  (:Is%20Valid
                    (for _array_element (Class|FortMinigameTeamComponent|GetTeamMembers (Minigame|GetMinigameTeamComponent _returnvalue) (Team|GetActorTeam _playerstate))
                      (bind _asfort_player_state_athena (Utilities|Casting|CastToFortPlayerStateAthena _array_element))
                      (if (or (Variables|GetbIncludeSelf) (Utilities|NotEqual(Object) _asfort_player_state_athena (Variables|/Script/Engine.Pawn|GetPlayerState (Variables|GetTarget))))
                        (Utilities|Array|Add (LocalVariables|GetPlayerStates) _asfort_player_state_athena)))
                    (return (LocalVariables|GetPlayerStates)))
                  (:Is%20Not%20Valid
                    _)))
              (:Is%20Not%20Valid
                _))
            (return 0))
          (:Is%20Not%20Valid
            _))
        (else
          ;[0 0.4344 1] Non Creative Playlists
          (bind _asfort_player_state_athena_2 (Utilities|Casting|CastToFortPlayerStateAthena (Variables|/Script/Engine.Pawn|GetPlayerState (Variables|GetTarget)))
            (:then
              (bind _returnvalue_2 (Class|FortGameStateAthena|GetSquadMemberPlayerStates _asfort_game_state_athena _asfort_player_state_athena_2 (Variables|GetbIncludeSelf)))
              (return _returnvalue_2))
            (:CastFailed
              _)))))
    (:CastFailed
      _))
  (return 0))

(fn InitAbility
  (Camera|PushCameraMode (Variables|Default|GetCameraMode))
  (CallFunction|ToggleReticleVisibility true)
  (Variables|Default|SetEnding false)
  (Utilities|Array|Clear (Variables|Default|GetActorsToMark))
  (Utilities|Array|Clear (Variables|Marking|GetSquadmatesToMark))
  (bind (_event_data _fort_player_pawn) (Setup|SetupGenericAbility false false)
    (:Output
      (bind _output_get (Variables|Default|SetPlayerPawn _fort_player_pawn))
      (Weapon|SetHolsterWeapon _output_get 0 true false (Variables|Default|GetRevealHolster))
      (bind _asfort_player_controller_athena (Utilities|Casting|CastToFortPlayerControllerAthena (Fortnite|GetFortPlayerControllerfromActor _output_get)))
      (Variables|Default|SetPlayerController _asfort_player_controller_athena))))

(fn Reveal
  (bind (_groupidentifier _duration _steptime _indicatoroffset _indicatordbnooffset _shareactorwith _shareactorwithmask) (Utilities|Struct|BreakIndicatedActorData (Variables|Default|GetIndicatedActorDataPlayers)))
  (bind (_groupidentifier_2 _duration_2 _steptime_2 _shareactorwith_2 _shareactorwithmask_2) (Utilities|Struct|BreakStenciledActorData (Variables|Default|GetStenciledActorDataPlayers)))
  (bind _markdurationplayer (Variables|Marking|GetMarkDurationPlayer))
  (bind _returnvalue (ScalableFloat|GetValueatLevel _markdurationplayer))
  (bind (_inaffiliations _returnvalue_2) (IndicatedManagement|MakeIndicatedActorDatawithFilter (Fortnite|Factions|GetFactionTagContainerforActor (Variables|Default|GetInterrogatedPlayer)) (Variables|Marking|GetNPCGameplayTagQuery) (Utilities|Array|MakeArray "ObjectTypeQuery3") "/Script/FortniteGame.FortPlayerPawnAthena" (IndicatedManagement|MakeIndicatedActorData :GroupIdentifier (Variables|Marking|GetIndicatorGroup) :ParticleSystemData (IndicatedManagement|MakeIndicatedActorParticleSystemData (Variables|Marking|GetIndicatedPS) (Variables|Marking|GetParticleSystemParamName) (Variables|Marking|GetPSOffset) (Variables|Marking|GetPSDBNOOffset) (Variables|Marking|GetPSVectorParamName)) :EventTags (Variables|Default|GetT_QuestTagContainer) :Duration (ScalableFloat|GetValueatLevel (Variables|Marking|GetMarkDurationNPC)) :StepTime (ScalableFloat|GetValueatLevel (Variables|Marking|GetIndicatorStepTime)) :IndicatorOffset "0.000000,0.000000,90.000000" :IndicatorDBNOOffset "0.000000,0.000000,30.000000" :ShareActorWith "SquadOnTeam" :Sound (Variables|Marking|GetIndicatedSound) :StateImageOverride "None" :StateIndexOverride -1 :bClampToScreen false :bHideStateImage false :bOnlyDisplayWhenStencilingIsDisabled false :bOnlyDisplayIfActorIsVisible false) (IndicatedManagement|MakeStenciledActorData :GroupIdentifier (Variables|Marking|GetIndicatorGroup) :Duration 8.0 :StepTime (ScalableFloat|GetValueatLevel (Variables|Marking|GetStencilStepTime)) :ShareActorWith "SquadOnTeam" :FriendlyStencilName "None" :EnemyStencilName (select (Class|AthenaFunctionLibrary|RowtoBool (Variables|Marking|GetbUseDefaultStencil)) "RedOverlayID" (Variables|Marking|GetEnemyStencilName)) :bOnlyDisplayIfActorIsVisible false) (ScalableFloat|GetValueatLevel (Variables|Marking|GetMarkRadius))))
  (bind (_inaffiliations_2 _returnvalue_3) (IndicatedManagement|MakeIndicatedActorDatawithFilter :IndicateActorTags (Variables|Marking|GetChestTagsToMark) :ObjectTypes (Utilities|Array|MakeArray "ObjectTypeQuery1") :ActorClassFilter "/Script/FortniteGame.BuildingContainer" :IndicatedActorData (IndicatedManagement|MakeIndicatedActorData :GroupIdentifier (Variables|Marking|GetIndicatorGroup) :EventTags (Variables|Default|GetT_QuestTagContainer) :Duration (ScalableFloat|GetValueatLevel (Variables|Marking|GetMarkDurationNPC)) :StepTime (ScalableFloat|GetValueatLevel (Variables|Marking|GetIndicatorStepTime)) :IndicatorOffset "0.000000,0.000000,90.000000" :IndicatorDBNOOffset "0.000000,0.000000,30.000000" :ShareActorWith "SquadOnTeam" :Sound (Variables|Marking|GetIndicatedSound) :StateImageOverride "None" :StateIndexOverride -1 :bClampToScreen false :bHideStateImage false :bOnlyDisplayWhenStencilingIsDisabled false :bOnlyDisplayIfActorIsVisible false) :StenciledActorData (IndicatedManagement|MakeStenciledActorData :GroupIdentifier (Variables|Marking|GetIndicatorGroup) :Duration 8.0 :StepTime (ScalableFloat|GetValueatLevel (Variables|Marking|GetStencilStepTime)) :ShareActorWith "SquadOnTeam" :FriendlyStencilName "None" :EnemyStencilName (select (Class|AthenaFunctionLibrary|RowtoBool (Variables|Marking|GetbUseDefaultStencil)) "YellowOverlayID" (Variables|Marking|GetTreasureChestStencilName)) :bOnlyDisplayIfActorIsVisible false) :OverlapRadius (ScalableFloat|GetValueatLevel (Variables|Marking|GetMarkRadius)) :OverlapHalfHeight 0.0))
  (bind (_inaffiliations_3 _returnvalue_4) (IndicatedManagement|MakeIndicatedActorDatawithFilter :IndicateActorTags (Variables|Marking|GetGameplayItemTagsToMark) :ObjectTypes (Utilities|Array|MakeArray "ObjectTypeQuery1" "ObjectTypeQuery2") :ActorClassFilter "/Script/FortniteGame.BuildingGameplayActor" :IndicatedActorData (IndicatedManagement|MakeIndicatedActorData :GroupIdentifier (Variables|Marking|GetIndicatorGroup) :EventTags (Variables|Default|GetT_QuestTagContainer) :Duration (ScalableFloat|GetValueatLevel (Variables|Marking|GetMarkDurationNPC)) :StepTime (ScalableFloat|GetValueatLevel (Variables|Marking|GetIndicatorStepTime)) :IndicatorOffset "0.000000,0.000000,90.000000" :IndicatorDBNOOffset "0.000000,0.000000,30.000000" :ShareActorWith "SquadOnTeam" :Sound (Variables|Marking|GetIndicatedSound) :StateImageOverride "None" :StateIndexOverride -1 :bClampToScreen false :bHideStateImage false :bOnlyDisplayWhenStencilingIsDisabled false :bOnlyDisplayIfActorIsVisible false) :StenciledActorData (IndicatedManagement|MakeStenciledActorData :GroupIdentifier (Variables|Marking|GetIndicatorGroup) :Duration 8.0 :StepTime 0.0 :ShareActorWith "SquadOnTeam" :FriendlyStencilName "None" :EnemyStencilName (select (Class|AthenaFunctionLibrary|RowtoBool (Variables|Marking|GetbUseDefaultStencil)) "RedOverlayID" (Variables|Marking|GetEnemyStencilName)) :bOnlyDisplayIfActorIsVisible false) :OverlapRadius (ScalableFloat|GetValueatLevel (Variables|Marking|GetMarkRadius)) :OverlapHalfHeight 0.0))
  (bind _playercontroller (Variables|Default|GetPlayerController))
  (bind _returnvalue_5 (Team|GetActorTeam _playercontroller))
  (bind _playerpawn (Variables|Default|GetPlayerPawn))
  (bind _squadmatestomark (Variables|Marking|GetSquadmatesToMark))
  (bind _indicatorgroup (Variables|Marking|GetIndicatorGroup))
  (bind _interrogatedplayer (Variables|Default|GetInterrogatedPlayer))
  (Utilities|FlowControl|Sequence
    (:then_0
      (Networking|CanExecuteCosmeticEvents
        (:True
          (Utilities|IsValid _playerpawn
            (:Is%20Valid
              (GameplayCue|ExecuteGameplayCue%28Burst%29Local _playerpawn (Variables|Default|GetGC_Reveal) (Ability|GameplayCue|MakeGameplayCueParameters :NormalizedMagnitude 0.0 :RawMagnitude 0.0 :Location "0,0,0" :Normal "0,0,0" :SourceObject _playerpawn :GameplayEffectLevel 1 :AbilityLevel 1 :bReplicateLocationWhenUsingMinimalRepProxy false)))))
        (:False
          (bind _returnvalue_6 (CallFunction|GetTeamMembers (Variables|Default|GetPlayerPawn)))
          (for _array_element _returnvalue_6
            (GameplayEffects|FortApplyGameplayEffectToTarget (Ability|GetFortAbilitySystemComponent (Variables|Default|GetPlayerPawn)) "/Game/Abilities/Player/Interrogation/GE_Interrogate_GC.GE_Interrogate_GC_C" (Ability|GetFortAbilitySystemComponent (PlayerState|GetCurrentPawn _array_element)) 0 0 0 0.0 (GameplayEffects|MakeFortEffectContext (Ability|GetFortAbilitySystemComponent (PlayerState|GetCurrentPawn _array_element)) (Variables|Default|GetPlayerPawn) (Variables|Default|GetPlayerPawn) (Variables|Default|GetPlayerPawn) 0))))))
    (:then_1
      (Utilities|IsValid _interrogatedplayer
        (:Is%20Valid
          (if (not (GameplayTags|HasMatchingGameplayTag _interrogatedplayer (Variables|Default|GetT_NPC)))
            (bind _returnvalue_7 (CallFunction|GetTeamMembers (Variables|Default|GetInterrogatedPlayer) true))
            (if (Utilities|IsArrayNonEmpty _returnvalue_7)
              (for _array_element_2 _returnvalue_7
                (Utilities|IsValid _array_element_2
                  (:Is%20Valid
                    (bind _returnvalue_8 (GameplayTags|HasAnyMatchingGameplayTags(Message) (PlayerState|GetPlayerPawnorVehicleDriver _array_element_2) (GameplayTags|MakeLiteralGameplayTagContainer "(GameplayTags=((TagName=\"Ability.Interrogation.Reveal.Immunity\")))")))
                    (if (not _returnvalue_8)
                      (Utilities|Array|Add (Variables|Marking|GetSquadmatesToMark) (PlayerState|GetPlayerPawnorVehicleDriver _array_element_2))))))
              (IndicatedManagement|AddActorsToIndicatedList _playercontroller _squadmatestomark (Utilities|Struct|MakeIndicatedActorData _indicatorgroup _returnvalue (ScalableFloat|GetValueatLevel (Variables|Marking|GetIndicatorStepTime)) "0.000000,0.000000,90.000000" "0.000000,0.000000,30.000000" "None" _shareactorwithmask true _returnvalue_5 "(SmallSizeDistance=7500.000000,LargestSizeDistance=10.000000,SmallScale=0.500000,LargestScale=0.800000,FarAwayScale=0.350000,FarAwayOpacity=1.000000,MaxScaleAndFadePercent=0.100000)" (IndicatedManagement|MakeIndicatedActorParticleSystemData (Variables|Marking|GetIndicatedPS) (Variables|Marking|GetParticleSystemParamName) (Variables|Marking|GetPSOffset) (Variables|Marking|GetPSDBNOOffset) (Variables|Marking|GetPSVectorParamName)) (Variables|Marking|GetIndicatedSound) "None" -1 true false false false (EventDispatchers|CreateEvent :!event GAB_InterrogatePlayer_Reveal_AutoGenFunc) (EventDispatchers|CreateEvent :!event GAB_InterrogatePlayer_Reveal_AutoGenFunc)))
              (IndicatedManagement|AddActorsToStenciledList _playercontroller _squadmatestomark (Utilities|Struct|MakeStenciledActorData _indicatorgroup _returnvalue (ScalableFloat|GetValueatLevel (Variables|Marking|GetStencilStepTime)) _shareactorwithmask_2 true _returnvalue_5 "(ParticleSystem=None,ParticleSystemActorParamName=\"\",ParticleSystemOffset=(X=0.000000,Y=0.000000,Z=0.000000),ParticleSystemDBNOOffset=(X=0.000000,Y=0.000000,Z=0.000000),ParticleSystemVectorParamName=\"\")" 0 "None" (select (Class|AthenaFunctionLibrary|RowtoBool (Variables|Marking|GetbUseDefaultStencil)) "RedOverlayID" (Variables|Marking|GetEnemyStencilName))))
              ;[1 1 1] Update Quest for Interrogating Player
              (Quest|SendComplexCustomStatEvent (Quests|GetQuestManager (Fortnite|GetFortPlayerControllerfromActor (Variables|Default|GetPlayerPawn)) "Invalid") 0 "(GameplayTags=((TagName=\"Athena.Quests.Interrogate.Player\")))" "(GameplayTags=)" 1))
            (else
              (IndicatedManagement|AddActorsInRadiusToIndicatedList (Variables|Default|GetPlayerController) (Utilities|Array|MakeArray _returnvalue_2 _returnvalue_3 _returnvalue_4))
              ;[1 1 1] Update Quest For Interrogating Mang
              (Quest|SendComplexCustomStatEvent (Quests|GetQuestManager (Fortnite|GetFortPlayerControllerfromActor (Variables|Default|GetPlayerPawn)) "Invalid") 0 (Variables|Default|GetT_QuestTagContainer) (Variables|/Script/FortniteGame.FortPawn|GetGameplayTags _interrogatedplayer) 1))))))))

(fn ToggleReticleVisibility
  (params (Hide :type Boolean))
  (HUD|SetHUDElementVisibility (BlueprintContext|GetFortHUDContext) (Variables|Default|GetT_ReticleHUD) Hide "None"))
