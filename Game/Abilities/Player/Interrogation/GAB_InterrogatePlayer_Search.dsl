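; Player gameplay ability for interrogating a hoisted DBNO pawn, such as a henchman, with a search montage.
; Setup holsters the weapon, hides the reticle and locks movement, while input press or either pawn's death ends the ability.
; The montage's Release tag applies a cooldown effect to the captive and sends the Reveal event, then cleanup restores control and drops the carry.
; Non-cosmetic loot and tag-matched keycard drop functions are defined but not called here.
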(event Custom|OnDied_Event_0
  (params (DamagedActor :type Actor%20Object%20Reference) (Damage :type Float%20%28single-precision%29) (InstigatedBy :type Controller%20Object%20Reference) (DamageCauser :type Actor%20Object%20Reference) (HitLocation :type Vector) (FHitComponent :type Primitive%20Component%20Object%20Reference) (BoneName :type Name) (Momentum :type Vector))
  ;[1 1 1] End the ability if either player dies
  (CallFunction|ToggleReticleVisibility false)
  (Ability|EndAbility))

(event Custom|OnDied_Event_1
  (params (DamagedActor :type Actor%20Object%20Reference) (Damage :type Float%20%28single-precision%29) (InstigatedBy :type Controller%20Object%20Reference) (DamageCauser :type Actor%20Object%20Reference) (HitLocation :type Vector) (FHitComponent :type Primitive%20Component%20Object%20Reference) (BoneName :type Name) (Momentum :type Vector))
  ;[1 1 1] End the ability if either player dies
  (CallFunction|ToggleReticleVisibility false)
  (Ability|EndAbility))

(event Ability|EventActivateAbility
  (bind _playerpawn (Variables|Default|GetPlayerPawn))
  (bind _returnvalue (Class|FortPlayerPawn|GetDBNOHoistee _playerpawn))
  (bind _interrogatedplayer (Variables|Default|GetInterrogatedPlayer))
  (bind _playerpawn_2 (Variables|Default|GetPlayerPawn))
  ;[0.3672 1 0.3979] Start Ability
  (Analytics|IncrementAnalyticMatchCount "HenchmenInterrogated")
  (|CallParentFunction)
  (CallFunction|InitAbility)
  (Utilities|FlowControl|Sequence
    (:then_0
      ;[1 1 1] End the ability if either player dies
      (Utilities|IsValid _interrogatedplayer
        (:Is%20Valid
          (EventDispatchers|AssignOnDied _interrogatedplayer (delegate Custom|OnDied_Event_0))
          (Utilities|IsValid _playerpawn_2
            (:Is%20Valid
              (EventDispatchers|AssignOnDied _playerpawn_2 (delegate Custom|OnDied_Event_1))
              (Ability|Tasks|WaitGameplayTagAdd "(TagName=\"MovementMode.Skydiving\")" _playerpawn_2 true
                (:Added
                  _)))))))
    (:then_1
      ;[1 1 1] Play Interrogation Synced Montage
      (bind (_asynctaskproxy _targetdata _applicationtag) (Ability|Tasks|PlayMontageandWaitforTrigger "None" (Variables|Default|GetAnim_Search))
        (:then
          (CallFunction|ApplyGameplayEffecttoNPC (Variables|Default|GetGE_PickedUp))
          ;[0 0 0] This lets us cancel while still in the animation. Since this is a UFortGameplayAbility_Action, there is a class default property called Action Name. We could set up the action names that we want to associate with this ability, so when input is pressed/released we can do what we want.
          (Ability|Tasks|WaitInputPress false
            (:OnPress
              (CallFunction|ToggleReticleVisibility false)
              (Ability|EndAbility))))
        (:Triggered
          (switch Utilities|FlowControl|Switch|SwitchonGameplayTag _applicationtag
            (:Ability.Interrogation.Release
              ;[1 1 1] Apply Cooldown, Trigger Reveal, and End Ability
              (Utilities|IsValid _playerpawn
                (:Is%20Valid
                  (GameplayEffects|ApplyGameplayEffectToTarget :self (Ability|GetAbilitySystemComponent _playerpawn) :GameplayEffectClass (Variables|Default|GetGE_Cooldown) :Target (Ability|GetAbilitySystemComponent _returnvalue) :Level 0.0)
                  (Ability|SendGameplayEventtoActor _playerpawn "(TagName=\"Ability.Interrogation.Reveal.Trigger\")" (Utilities|Struct|MakeGameplayEventData "(TagName=\"Ability.Interrogation.Reveal.Trigger\")" 0 0 _returnvalue 0 "()" "(GameplayTags=((TagName=\"Pawn.Athena.NPC.Mang\")))"))
                  (Ability|EndAbility))))))
        (:Cancelled
          _))))
  (CallFunction|ToggleReticleVisibility false)
  (Ability|EndAbility))

(event Ability|EventOnEndAbility
  (params (bWasCancelled :type Boolean))
  ;[1 0.4227 0.01201] End Ability
  (|CallParentFunction bWasCancelled)
  (if bWasCancelled
    (CallFunction|ToggleReticleVisibility false))
  (Default|EndAbilityCleanup))

(fn ApplyGameplayEffectToNPC
  (params (GameplayEffect :type Gameplay%20Effect%20Class%20Reference))
  (bind _playerpawn (Variables|Default|GetPlayerPawn))
  (bind _returnvalue (Class|FortPlayerPawn|GetDBNOHoistee _playerpawn))
  (Networking|CanExecuteCosmeticEvents
    (:False
      (Utilities|IsValid _playerpawn
        (:Is%20Valid
          (Utilities|IsValid _returnvalue
            (:Is%20Valid
              (if (GameplayTags|HasMatchingGameplayTag _returnvalue (Variables|Loot|GetT_NPC))
                (GameplayEffects|ApplyGameplayEffectToSelf :self (Ability|GetAbilitySystemComponent _returnvalue) :GameplayEffectClass GameplayEffect :Level 0.0)))))))))

(fn DropKeycard
  (bind (_haskeycard _output) (CallFunction|GetKeycardtoDrop))
  (bind _interrogatedplayer (Variables|Default|GetInterrogatedPlayer))
  (bind _playerpawn (Variables|Default|GetPlayerPawn))
  (Networking|CanExecuteCosmeticEvents
    (:False
      (if (> (Math|Float|Truncate (ScalableFloat|GetValueatLevel (Variables|Loot|GetMaxDropsInstance))) (Variables|Loot|GetNumDropsInstance))
        (if (Math|Random|RandomBoolwithWeight (ScalableFloat|GetValueatLevel (Variables|Loot|GetKeycardDropBias)))
          (if _haskeycard
            (Variables|Loot|SetItemDefinition _output)
            (Variables|Loot|SetNumberToSpawn 1)
            (Variables|Loot|SetKeycardDropBias)
            (Utilities|IsValid _playerpawn
              (:Is%20Valid
                (Utilities|IsValid _interrogatedplayer
                  (:Is%20Valid
                    (Inventory|SpawnPickupInWorld (Variables|Loot|GetItemDefinition) (Variables|Loot|GetNumberToSpawn) (+ (Transformation|GetActorLocation _interrogatedplayer) (Variables|Loot|GetLootSpawnOffset)) (Transformation|GetActorForwardVector _playerpawn))
                    (Math|Integer|IncrementInt (Variables|Loot|GetNumDropsInstance))))))))))))

(fn DropLoot
  (bind _interrogatedplayer (Variables|Default|GetInterrogatedPlayer))
  (bind _playerpawn (Variables|Default|GetPlayerPawn))
  (Networking|CanExecuteCosmeticEvents
    (:False
      (if (> (Math|Float|Truncate (ScalableFloat|GetValueatLevel (Variables|Loot|GetMaxDropsInstance))) (Variables|Loot|GetNumDropsInstance))
        (bind (_outloottodrop _returnvalue) (Loot|PickLootDrops (Variables|Loot|GetLootTierGroupName) 1 -1 "(GameplayTags=)"))
        (if _returnvalue
          (for _array_element _outloottodrop
            (bind (_array_element_itemdefinition _array_element_itemguid _array_element_count) (Class|FortItemFunctionLibrary|BreakFortItemEntry _array_element))
            (bind _asfort_world_item_definition (Utilities|Casting|CastToFortWorldItemDefinition _array_element_itemdefinition))
            (Variables|Loot|SetItemDefinition _asfort_world_item_definition)
            (Variables|Loot|SetNumberToSpawn _array_element_count)
            (Utilities|IsValid _playerpawn
              (:Is%20Valid
                (Utilities|IsValid _interrogatedplayer
                  (:Is%20Valid
                    (Inventory|SpawnPickupInWorld (Variables|Loot|GetItemDefinition) (Variables|Loot|GetNumberToSpawn) (+ (Transformation|GetActorLocation _interrogatedplayer) (Variables|Loot|GetLootSpawnOffset)) (Transformation|GetActorForwardVector _playerpawn))
                    (Math|Integer|IncrementInt (Variables|Loot|GetNumDropsInstance))))))))))))

(fn EndAbilityCleanup
  (bind _playerpawn (Variables|Default|GetPlayerPawn))
  (bind _playercontroller (Variables|Default|GetPlayerController))
  (Utilities|Time|ClearandInvalidateTimerbyHandle (Variables|Default|GetTimer_Interrogation))
  (Utilities|FlowControl|Sequence
    (:then_0
      (Utilities|IsValid _playerpawn
        (:Is%20Valid
          (Weapon|SetHolsterWeapon _playerpawn 0 false false (Variables|Default|GetInterrogateHolster))
          (Class|FortPlayerPawn|StopCarryingPlayer _playerpawn false))))
    (:then_1
      (Utilities|IsValid _playercontroller
        (:Is%20Valid
          (Input|SetRotatePawntoCamera _playercontroller true)
          (Input|SetIgnoreMoveInput _playercontroller false))))))

(fn GetKeycardToDrop
  (bind _keycardindex (LocalVariables|GetKeycardIndex))
  (bind _keycarditemdefinitions (Variables|Loot|GetKeycardItemDefinitions))
  (bind _output (Utilities|Array|Get(acopy) _keycarditemdefinitions _keycardindex))
  (bind _interrogatedplayer (Variables|Default|GetInterrogatedPlayer))
  (Utilities|IsValid _interrogatedplayer
    (:Is%20Valid
      (if (GameplayTags|HasMatchingGameplayTag _interrogatedplayer (Variables|Loot|GetT_NPC))
        (if (GameplayTags|HasMatchingGameplayTag _interrogatedplayer (Variables|Loot|GetT_Keycard_Yacht))
          (LocalVariables|SetKeycardIndex 0)
          (elif (GameplayTags|HasMatchingGameplayTag _interrogatedplayer (Variables|Loot|GetT_Keycard_OilRig))
            (LocalVariables|SetKeycardIndex 1)
            (elif (GameplayTags|HasMatchingGameplayTag _interrogatedplayer (Variables|Loot|GetT_Keycard_SharkIsland))
              (LocalVariables|SetKeycardIndex 2)
              (elif (GameplayTags|HasMatchingGameplayTag _interrogatedplayer (Variables|Loot|GetT_Keycard_HQ))
                (LocalVariables|SetKeycardIndex 3)
                (elif (GameplayTags|HasMatchingGameplayTag _interrogatedplayer (Variables|Loot|GetT_Keycard_MountainBase))
                  (LocalVariables|SetKeycardIndex 4)
                  (else
                    (return false 0)))))))
        (Utilities|IsValid _output
          (:Is%20Valid
            (return true _output))
          (:Is%20Not%20Valid
            _)))
      (return false 0))))

(fn InitAbility
  (bind (_event_data _fort_player_pawn) (Setup|SetupGenericAbility false false)
    (:Output
      (bind _output_get (Variables|Default|SetPlayerPawn _fort_player_pawn))
      (Variables|Default|SetInterrogatedPlayer (Class|FortPlayerPawn|GetDBNOHoistee _output_get))
      (Weapon|SetHolsterWeapon _output_get 0 true false (Variables|Default|GetInterrogateHolster))
      (CallFunction|ToggleReticleVisibility true)
      (Variables|Loot|SetNumDropsInstance 0)
      (bind _asfort_player_controller_athena (Utilities|Casting|CastToFortPlayerControllerAthena (Fortnite|GetFortPlayerControllerfromActor _output_get)))
      (bind _output_get_2 (Variables|Default|SetPlayerController _asfort_player_controller_athena))
      (Input|SetIgnoreMoveInput _output_get_2 true)
      (Input|SetRotatePawntoCamera _output_get_2 false))))

(fn ToggleReticleVisibility
  (params (Hide :type Boolean))
  (HUD|SetHUDElementVisibility (BlueprintContext|GetFortHUDContext) (Variables|Default|GetT_ReticleHUD) Hide "None"))
