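; Gameplay ability in the player interrogation set that plays a gameplay cue on non-server machines.
; On activation it reads the source object from the granting effect's context.
; Off the server, it executes the Cue variable as a local burst on the avatar with that source object in the parameters, then ends.
; The server branch never calls EndAbility, and debug prints remain.
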(event Ability|EventActivateAbility
  (bind _returnvalue (Ability|GetGrantedbyEffectContext))
  (bind _returnvalue_2 (Ability|EffectContext|GetSourceObject _returnvalue))
  (Development|PrintString "GC GA activated" true true "(R=0.000000,G=1.000000,B=0.085407,A=1.000000)" 1e+01)
  (if (not (Networking|IsServer))
    (GameplayCue|ExecuteGameplayCue%28Burst%29Local (Ability|GetAvatarActorfromActorInfo) (Variables|Default|GetCue) (Ability|GameplayCue|MakeGameplayCueParameters :NormalizedMagnitude 0.0 :RawMagnitude 0.0 :Location "0,0,0" :Normal "0,0,0" :SourceObject _returnvalue_2 :GameplayEffectLevel 1 :AbilityLevel 1 :bReplicateLocationWhenUsingMinimalRepProxy false))
    (Development|PrintString (Utilities|GetDisplayName _returnvalue_2) true true "(R=1.000000,G=0.000000,B=0.794666,A=1.000000)" 1e+01)
    (Ability|EndAbility)))

(event Ability|EventOnEndAbility
  (params (bWasCancelled :type Boolean)))

(event Ability|EventFailedToActivatePassiveAbility
  (params (ActorInfo :type Gameplay%20Ability%20Actor%20Info%20Structure))
  (|CallParentFunction))
